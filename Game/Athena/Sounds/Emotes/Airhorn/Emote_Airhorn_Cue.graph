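# Sound cue for the Airhorn emote in Athena, sequencing a step-in, a horn blast and a step-out.
# A Mixer starts the step-in immediately, one of three equally weighted random horn variants after half a second, and the step-out at about 2.7 seconds, using fixed Delay nodes.
# It uses emote attenuation and the emote music sound class, overrides concurrency to stop the oldest instance, and references a music audio-analysis submix send.
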SoundCue /Game/Athena/Sounds/Emotes/Airhorn/Emote_Airhorn_Cue
root=Mixer_0
nodes [
 {id=Delay_0, children [Random_0], properties {delayMax=0.5, delayMin=0.5}}
 {id=Delay_1, children [WavePlayer_4], properties {delayMax=2.694, delayMin=2.694}}
 {id=Mixer_0, children [WavePlayer_0,Delay_0,Delay_1], properties {inputVolume [1,1,1]}}
 {id=Random_0, children [WavePlayer_1,WavePlayer_2,WavePlayer_3], properties {bShouldExcludeFromBranchCulling=true, weights [1,1,1]}}
 {id=WavePlayer_0, properties {soundWaveAssetPtr=&/Emote_Airhorn_StepIn_01}}
 {id=WavePlayer_1, properties {soundWaveAssetPtr=&/Emote_Airhorn_03}}
 {id=WavePlayer_2, properties {soundWaveAssetPtr=&/Emote_Airhorn_01}}
 {id=WavePlayer_3, properties {soundWaveAssetPtr=&/Emote_Airhorn_02}}
 {id=WavePlayer_4, properties {soundWaveAssetPtr=&/Emote_Airhorn_StepOut_01}}
]
properties {
 attenuationSettings=&Fort_Emotes_Athena_Att
 bOverrideConcurrency=true
 maxDistance=1600
 soundClassObject=&Fort_Music_Emote
 volumeMultiplier=1
 concurrencyOverrides {maxCount=3, resolutionRule=StopOldest}
 concurrencySet [/Game/Sounds/Concurrency_Rules/Athena_EmoteMusic_Conc]
 soundSubmixSends [{minSendLevel=0, soundSubmix=&Fortnite_AudioAnalysis_Music}]
}
